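; Base AI controller Blueprint for IO Guard bots that adjusts the pawn's movement while it patrols.
; When patrolling starts, it applies a patrolling gameplay effect to the bot pawn's ability system and stores the handle.
; It also saves the character movement rotation rate and swaps in a patrolling rate.
; When patrolling stops, it removes that effect by handle and restores the saved rotation rate.
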(event OnStartPatrollingEvent(FortAthenaNpcPatrolling)
  (bind _playerbotpawn (Variables|FortAthenaAIBotController|GetPlayerBotPawn))
  (bind _returnvalue (Ability|GetAbilitySystemComponent _playerbotpawn))
  (bind _charactermovement (Variables|/Script/Engine.Character|GetCharacterMovement _playerbotpawn))
  (Utilities|IsValid _returnvalue
    (:Is%20Valid
      (bind _returnvalue_2 (GameplayEffects|ApplyGameplayEffectToTarget :self _returnvalue :GameplayEffectClass (Variables|Default|GetGE_Patrolling) :Target _returnvalue :Level 0.0))
      (Variables|Default|SetGE_PatrollingSpeedHandle _returnvalue_2)
      (Variables|Default|SetSavedRotationRate (Variables|/Script/Engine.CharacterMovementComponent|GetRotationRate _charactermovement))
      (Variables|/Script/Engine.CharacterMovementComponent|SetRotationRate (Variables|Default|GetPatrollingRotationRate) _charactermovement))))

(event OnStopPatrollingEvent(FortAthenaNpcPatrolling)
  (bind _playerbotpawn (Variables|FortAthenaAIBotController|GetPlayerBotPawn))
  (bind _returnvalue (Ability|GetAbilitySystemComponent _playerbotpawn))
  (Utilities|IsValid _returnvalue
    (:Is%20Valid
      (GameplayEffects|RemoveActiveGameplayEffect _returnvalue (Variables|Default|GetGE_PatrollingSpeedHandle) -1)
      (Variables|/Script/Engine.CharacterMovementComponent|SetRotationRate (Variables|Default|GetSavedRotationRate) (Variables|/Script/Engine.Character|GetCharacterMovement _playerbotpawn)))))

(fn ConstructionScript
  (|CallParentFunction))
